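# Looping ambient sound cue for the night interior of the Tomato Temple in the Jungle ambience set.
# A Mixer layers a looping night bed with a slightly boosted percussion branch that endlessly picks one of two one-shots at random, each preceded by a randomized delay.
# A quality-level root feeds the same mix to its first two slots and leaves the other two empty.
# It uses tile-ambient attenuation and the ambient tile-mix sound class.
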SoundCue /Figment_S08/Sounds/Fort_Ambient_Looping/Jungle/Ambient_TomatoTemple_Interior_Night_Cue
root=QualityLevel_0
nodes [
 {id=Delay_0, children [WavePlayer_0], properties {delayMax=22, delayMin=11}}
 {id=Delay_1, children [WavePlayer_2], properties {delayMax=22, delayMin=11}}
 {id=Looping_0, children [Random_1]}
 {id=Mixer_1, children [WavePlayer_1,Looping_0], properties {inputVolume [1,1.1]}}
 {id=QualityLevel_0, children [Mixer_1,Mixer_1,null,null]}
 {id=Random_1, children [Delay_0,Delay_1], properties {weights [1,1]}}
 {id=WavePlayer_0, properties {soundWaveAssetPtr=&/Ambient_TomatoTemple_Interior_Perc01}}
 {
  id=WavePlayer_1
  properties {bLooping=true, soundWaveAssetPtr=&/Ambient_TomatoTemple_Interior_Night_01}
 }
 {id=WavePlayer_2, properties {soundWaveAssetPtr=&/Ambient_TomatoTemple_Interior_Perc02}}
]
properties {
 attenuationSettings=/Game/Sounds/Attenuation/Fort_TileAmbient_Att
 maxDistance=10000
 soundClassObject=/Game/Sounds/FortSoundClassesAndModes/Fort_Ambient_TileMix
 volumeMultiplier=0.6
}
